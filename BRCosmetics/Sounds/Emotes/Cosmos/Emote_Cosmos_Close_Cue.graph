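# Sound cue for the close-up audio of the Cosmos emote, played as emote music.
# A single wave player outputs the pram close mechanical sound, spatialized by the close pets attenuation at reduced volume.
# An owner-limited concurrency override allows one instance and blocks new ones, alongside a shared emote foley concurrency rule.
# It also sends to the music audio analysis submix.
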SoundCue /BRCosmetics/Sounds/Emotes/Cosmos/Emote_Cosmos_Close_Cue
root=WavePlayer_0
nodes [{id=WavePlayer_0, properties {soundWaveAssetPtr=&/Emote_Cosmos_Pram_Close_A_Mech_01}}]
properties {
 attenuationSettings=&Fort_Pets_Close_Athena_Att
 maxDistance=1800.0001
 soundClassObject=&Fort_Music_Emote
 volumeMultiplier=0.336
 concurrencyOverrides {bLimitToOwner=true, maxCount=1, resolutionRule=PreventNew}
 concurrencySet [/Game/Sounds/Concurrency_Rules/Athena_EmoteFoley_Conc]
 soundSubmixSends [{minSendLevel=0, soundSubmix=&Fortnite_AudioAnalysis_Music}]
}
